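# Niagara system for the Despair pickaxe's swing trail, drawing a short-lived ribbon behind the melee weapon as it animates.
# A single FN_AnimTrail emitter spawns and updates ribbon particles from bone or socket positions on the User.SkeletalMesh, with ribbon ID offset support.
# A color curve turns the trail from bright to darker red while fading alpha, with RGB scaled by time of day.
# The second bone pair, velocity and U-override steps are disabled.
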NiagaraSystem /BRCosmetics/Weapons/FORT_Melee/Pickaxe_Despair/FX/NS_Pickaxe_Despair_Swing
dependencies {usedDataInterfaces [&NiagaraDataInterfaceColorCurve], usedDynamicInputs [], usedModules [&SystemState]}
emitters {FN_AnimTrail {
  emitterSpawnScript [&FN_RibbonIDOffsetUtility()]
  emitterUpdateScript [
   &EmitterState(Life Cycle Mode=::NewEnumerator1, Loop Behavior=::NewEnumerator1, Loop Duration=0.3, UseLoopDelay=false)
   &FN_SpawnAnimTrail(AnimTrailWidthMode=::FromSecond, Spawn Interval=0)
  ]
  particleSpawnScript [
   &FN_InitializeAnimTrailBoneIndecies()
   &FN_InitializeAnimTrailBoneIndecies(AnimTrail Selection=1, First Bone or Socket Index=2, Second Bone or Socket Index=3) {enabled=false, module=FN_InitializeAnimTrailBoneIndecies001}
   &FN_InitializeAnimTrail(Lifetime=0.2)
   &AddVelocity(Velocity=<0 400 0>) {enabled=false}
  ]
  particleUpdateScript [
   &SolveForcesAndVelocity(Clamp Velocity=false, Limit Acceleration=false) {enabled=false}
   &FN_UpdateAnimTrail(Locked Particle RibbonWidth=1, SkeletalMesh=$User.SkeletalMesh) {module=FN_UpdateAnimTrail001}
   /SetVariables(Particles.RibbonU0Override=$Particles.NormalizedAge) {enabled=false, isSetParameters=true}
   &ScaleColor(
    Scale RGB=&Multiply_VectorByFloat(
     Vector=&MakeVectorFromLinearColorRGB(Color=&ColorFromCurve(ColorCurve=DataInterface() {propertyValues {
         AlphaCurve {keys [<at 0.0685 1>,<at 0.2395 0.5>,<at 0.8992 0>]}
         BlueCurve {keys [<at 0 0>,<at 0.6166 0.0104>]}
         GreenCurve {keys [<at 0 0.0468>,<at 0.6166 0>]}
         RedCurve {keys [<at 0 1>,<at 0.6166 0.5>]}
        }}))
     Float=&FNTimeOfDayScalar(Night Scale=0.3, Morning Scale=0.5, Day Scale=0.7, Evening Scale=0.5)
    )
    ScaleA=true
    ScaleRGB=true
   ) {module=ScaleColor001}
   &SpriteFacingAndAlignment(Sprite Facing=$Particles.RibbonFacing, WriteAlignment=false, WriteFacing=true)
  ]
  rendererClasses [&NiagaraRibbonRendererProperties]
  renderers [{rendererClass=&NiagaraRibbonRendererProperties}]
 }}
summary {
 systemName=NS_Pickaxe_Despair_Swing
 emitters [{emitterName=FN_AnimTrail, rendererClasses [&NiagaraRibbonRendererProperties]}]
 userVariables [
  User.RibbonIDOffset {type=NiagaraInt32}
  User.SkeletalMesh {
   type=NiagaraDataInterfaceSkeletalMesh
   defaultValue {
    struct=NiagaraExt_VariableValue_DataInterface
    value {dataInterface=&self:NiagaraDataInterfaceSkeletalMesh_0}
   }
  }
 ]
}
